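# Animation Blueprint for the WaterMolecules backpack, which plays open and close animations as the matching pickaxe is equipped or put away.
# Each update sets IsCitadelPickaxeEquipped by substring-matching the current weapon's object name, driving an Idle, Equipped, Unequipped state machine with zero-length crossfades.
# Spine_05's yaw delta from ref pose, relative to spine_01, is mapped to an additive root-bone roll and Y offset after the state machine.
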AnimBlueprint /BRCosmetics/Accessories/FORT_Backpacks/Backpack_M_MED_WaterMolecules/Meshes/M_MED_WaterMolecules_Pack_AnimBP
target_skeleton=&/M_MED_WaterMolecules_Pack_Skeleton
graphs [
 AnimGraph {nodes [
   Input Pose()
   AnimGraphNode_StateMachine_0 = PlayShieldAnim() {is=title AnimGraphNode_StateMachine}
   n1 = "Transform (Modify) Bone - Bone: root"(
    ComponentPose=Local To Component(LocalPose=@AnimGraphNode_StateMachine_0)
    Translation=Get backpack_translation()
    Rotation=Get backpack_rotate()
   ) {
    rotationMode=BMM_Additive
    rotationSpace=BCS_BoneSpace
    translationMode=BMM_Additive
    translationSpace=BCS_BoneSpace
    alphaBoolBlend {blendOption=Linear}
    boneToModify {boneName=root}
   }
   Output Pose(Result=Component To Local(ComponentPose=@n1)) {layerGroup=None}
  ]}
 PlayShieldAnim {
  entry=Idle
  path=AnimGraph.AnimGraphNode_StateMachine_0.PlayShieldAnim
  ids=id
  nodes [
   AnimStateNode_0 = Idle()
   AnimStateNode_1 = Equipped() {is=title AnimStateNode}
   AnimStateNode_2 = Unequipped() {is=title AnimStateNode}
   AnimStateTransitionNode_0 = AnimStateTransitionNode() {crossfadeDuration=0}
   AnimStateTransitionNode_1 = AnimStateTransitionNode() {crossfadeDuration=0}
   AnimStateTransitionNode_2 = AnimStateTransitionNode() {bAutomaticRuleBasedOnSequencePlayerInState=true, crossfadeDuration=0}
  ]
  links [
   {
    dst=Unequipped
    rule=AnimGraph.AnimGraphNode_StateMachine_0.PlayShieldAnim.AnimStateTransitionNode_1.Transition
    src=Equipped
    via=AnimStateTransitionNode_1
   }
   {
    dst=Equipped
    rule=AnimGraph.AnimGraphNode_StateMachine_0.PlayShieldAnim.AnimStateTransitionNode_0.Transition
    src=Idle
    via=AnimStateTransitionNode_0
   }
   {
    dst=Idle
    rule=AnimGraph.AnimGraphNode_StateMachine_0.PlayShieldAnim.AnimStateTransitionNode_2.Transition
    src=Unequipped
    via=AnimStateTransitionNode_2
   }
  ]
 }
 Idle {
  path=AnimGraph.AnimGraphNode_StateMachine_0.PlayShieldAnim.AnimStateNode_0.Idle
  nodes [Output Animation Pose(Result=Local Space Ref Pose())]
 }
 Equipped {
  path=AnimGraph.AnimGraphNode_StateMachine_0.PlayShieldAnim.AnimStateNode_1.Equipped
  nodes [
   n1 = Sequence Player 'WaterMolecules_Open_Pack'!(bLoopAnimation:bool=False) -> (Pose) {bLoopAnimation=false, sequence=/BRCosmetics/Accessories/FORT_Backpacks/Backpack_M_MED_WaterMolecules/Animation/WaterMolecules_Open_Pack}
   Output Animation Pose(Result=@n1)
  ]
 }
 Unequipped {
  path=AnimGraph.AnimGraphNode_StateMachine_0.PlayShieldAnim.AnimStateNode_2.Unequipped
  nodes [
   n1 = Sequence Player 'WaterMolecules_Close_Pack'!(bLoopAnimation:bool=False) -> (Pose) {bLoopAnimation=false, sequence=/BRCosmetics/Accessories/FORT_Backpacks/Backpack_M_MED_WaterMolecules/Animation/WaterMolecules_Close_Pack}
   Output Animation Pose(Result=@n1)
  ]
 }
 Transition {
  path=AnimGraph.AnimGraphNode_StateMachine_0.PlayShieldAnim.AnimStateTransitionNode_0.Transition
  nodes [Result(bCanEnterTransition=Get IsCitadelPickaxeEquipped() -> (IsCitadelPickaxeEquipped:bool))]
 }
 Transition {
  path=AnimGraph.AnimGraphNode_StateMachine_0.PlayShieldAnim.AnimStateTransitionNode_1.Transition
  nodes [Result(bCanEnterTransition=NOT Boolean(A=Get IsCitadelPickaxeEquipped()))]
 }
 Transition {path=AnimGraph.AnimGraphNode_StateMachine_0.PlayShieldAnim.AnimStateTransitionNode_2.Transition, nodes [Result()]}
 EventGraph {nodes [
   Event BlueprintInitializeAnimation()
   > n1 = Cast To FortPlayerPawn(Object=TryGetPawnOwner())
   > Set As Fort Player Pawn(As Fort Player Pawn:Fort Player Pawn Object Reference=@n1) -> (Output_Get:Fort Player Pawn Object Reference)
   Event BlueprintUpdateAnimation()
   > n2 = Sequence()
   n3 = GetCurrentWeapon!(self:Fort Pawn Object Reference=Get As Fort Player Pawn() -> (As Fort Player Pawn:Fort Player Pawn Object Reference)) -> (ReturnValue:Fort Weapon Object Reference)
   > Set IsCitadelPickaxeEquipped(IsCitadelPickaxeEquipped:bool=Contains(SearchIn="To String (Object)"!(InObj:Object Reference=@n3) -> (ReturnValue:String), Substring=B_Athena_Pickaxe_WaterMolecules)) -> (Output_Get:bool)
   n4 = GetOwnerSkeletalMeshComponent()
   n5 = BreakRotator(InRot=BreakTransform(InTransform=GetDeltaTransformFromRefPose(self=@n4, BoneName=spine_05, BaseName=spine_01)).Rotation)
   Set backpack_rotate(execute=@n2.then_1, backpack_rotate=MakeRotator(Roll=MapRangeClamped(Value=@n5.Yaw, InRangeA=0, InRangeB=-30, OutRangeB=30)))
   > Set backpack_translation(backpack_translation=MakeVector(Y=MapRangeClamped(Value=@n5.Yaw, InRangeA=0, InRangeB=-30, OutRangeB=-6)))
  ]}
]
variables [
 As Fort Player Pawn {type=FortPlayerPawn}
 backpack_rotate {type=Rotator}
 backpack_translation {type=Vector}
 IsCitadelPickaxeEquipped {type=boolean}
]
